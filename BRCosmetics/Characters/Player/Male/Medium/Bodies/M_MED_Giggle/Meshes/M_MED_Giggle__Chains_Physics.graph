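# Physics asset for the dangling chains on the M_MED_Giggle male body cosmetic, covering the right hand and shoulder.
# Capsules on pelvis, spine, right clavicle, upper arm and forearm twist bone form a body linked by free-rotation constraints.
# A single hand chain body hangs from the forearm and a two-segment shoulder chain hangs from the clavicle.
# Chain joints have swing limits, wider at the hand, and zero-degree twist limits.
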PhysicsAsset /BRCosmetics/Characters/Player/Male/Medium/Bodies/M_MED_Giggle/Meshes/M_MED_Giggle__Chains_Physics
bodies {
 clavicle_r [{
   shapeType=Capsule
   length=3.6365
   radius=5.046
   shapeName=clavicle_r_capsule
   center=<-11.9774 5.6195 -1.1305>
   rotation=<rot 23.6818 -90 9.0968>
  }]
 dyn_handChain [{
   shapeType=Capsule
   length=6.1873
   radius=1.8135
   shapeName=dyn_handChain_capsule
   center=<10.1364 0.0604 -0.1393>
   rotation=<rot 0 -89.9879 -89.176>
  }]
 dyn_shoulderChain1 [{
   shapeType=Capsule
   length=7.8919
   radius=1.8579
   shapeName=dyn_shoulderChain1_capsule
   center=<10.513 -0.2202 0.0393>
   rotation=<rot 0 -90.1322 -89.7268>
  }]
 dyn_shoulderChain2 [{
   shapeType=Capsule
   length=4.5302
   radius=1.973
   shapeName=dyn_shoulderChain2_capsule
   center=<11.9731 -0.1465 0.0265>
   rotation=<rot 0 87.2747 -87.7958>
  }]
 lowerarm_twist_02_r [{
   shapeType=Capsule
   length=10.2085
   radius=6.0546
   shapeName=lowerarm_twist_02_r_capsule
   center=<1.0104 -1.2843 0.1864>
   rotation=<rot -30.4661 -90 -91.034>
  }]
 pelvis [{shapeType=Capsule, length=7.5, radius=11.25, shapeName=pelvis_capsule}]
 spine_02 [{shapeType=Capsule, length=5.25, radius=13.1476, shapeName=spine_02_capsule, rotation=<rot -0.0022 -179.9999 -180>}]
 upperarm_r [{shapeType=Capsule, length=20.5, radius=5.5, shapeName=upperarm_r_capsule, center=<-11.3172 -0.0003 -0.0002>, rotation=<rot -90 0 0>}]
}
constraints [
 {
  bone1Name=lowerarm_twist_02_r
  bone2Name=upperarm_r
  swing1LimitDegrees=45
  swing1Motion=Free
  swing2LimitDegrees=45
  swing2Motion=Free
  twistLimitDegrees=45
  twistMotion=Free
 }
 {bone1Name=dyn_handChain, bone2Name=lowerarm_twist_02_r, swing1LimitDegrees=90, swing2LimitDegrees=90, twistLimitDegrees=0}
 {
  bone1Name=upperarm_r
  bone2Name=clavicle_r
  swing1LimitDegrees=45
  swing1Motion=Free
  swing2LimitDegrees=45
  swing2Motion=Free
  twistLimitDegrees=45
  twistMotion=Free
 }
 {
  bone1Name=clavicle_r
  bone2Name=spine_02
  swing1LimitDegrees=45
  swing1Motion=Free
  swing2LimitDegrees=45
  swing2Motion=Free
  twistLimitDegrees=45
  twistMotion=Free
 }
 {bone1Name=dyn_shoulderChain1, bone2Name=clavicle_r, swing1LimitDegrees=45, swing2LimitDegrees=45, twistLimitDegrees=0}
 {bone1Name=dyn_shoulderChain2, bone2Name=dyn_shoulderChain1, swing1LimitDegrees=45, swing2LimitDegrees=45, twistLimitDegrees=0}
 {
  bone1Name=spine_02
  bone2Name=pelvis
  swing1LimitDegrees=45
  swing1Motion=Free
  swing2LimitDegrees=45
  swing2Motion=Free
  twistLimitDegrees=45
  twistMotion=Free
 }
]
