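# Physics asset for the skirt of the medium male M_MED_Giggle player body, giving its dynamic bones collision and joints.
# Left, right and middle capsule chains hang from the pelvis, with the middle chain splitting at its tip and LC starting at its second bone.
# Pelvis, thigh and calf capsules cover the body, while oversized root and floorCOL capsules sit below, apparently as ground collision.
# Body joints and a few skirt joints rotate freely.
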PhysicsAsset /BRCosmetics/Characters/Player/Male/Medium/Bodies/M_MED_Giggle/Meshes/M_MED_Giggle__Skirt_PhysicsNormal
bodies {
 calf_l [{shapeType=Capsule, length=44.5671, radius=9.875, shapeName=calf_l_capsule, center=<-27.3164 -2.1209 0>, rotation=<rot 90 0 0>}]
 calf_r [{shapeType=Capsule, length=23.3588, radius=10.0591, shapeName=calf_r_capsule, center=<14.8023 2.1213 -0.0001>, rotation=<rot 90 0 0>}]
 dyn_skirtLA1 [{shapeType=Capsule, length=3, radius=3, shapeName=dyn_skirtLA1_capsule}]
 dyn_skirtLA2 [{shapeType=Capsule, length=7.75, radius=8.25, shapeName=dyn_skirtLA2_capsule, center=<10.258 -4 0>, rotation=<rot 90 0 0>}]
 dyn_skirtLA3 [{shapeType=Capsule, length=7, radius=9, shapeName=dyn_skirtLA3_capsule, center=<10.258 -3 0>, rotation=<rot 90 0 0>}]
 dyn_skirtLA4 [{shapeType=Capsule, length=23, radius=10, shapeName=dyn_skirtLA4_capsule, center=<26.0324 -3 -0.0006>}]
 dyn_skirtLB1 [{shapeType=Capsule, length=3, radius=3, shapeName=dyn_skirtLB1_capsule}]
 dyn_skirtLB2 [{shapeType=Capsule, length=9, radius=7, shapeName=dyn_skirtLB2_capsule, center=<10 -7 0.0003>, rotation=<rot 90 0 0>}]
 dyn_skirtLB3 [{shapeType=Capsule, length=8, radius=8, shapeName=dyn_skirtLB3_capsule, center=<10 -3 0>, rotation=<rot 90 0 0>}]
 dyn_skirtLB4 [{shapeType=Capsule, length=29, radius=10.75, shapeName=dyn_skirtLB4_capsule, center=<25.7744 -3 -0.0006>}]
 dyn_skirtLC2 [{
   shapeType=Capsule
   length=7.4088
   radius=4.5317
   shapeName=dyn_skirtLC2_capsule
   center=<14.5771 1.1883 -6.2809>
   rotation=<rot -8.4648 -90 -38.3918>
  }]
 dyn_skirtLC3 [{shapeType=Capsule, length=8, radius=8, shapeName=dyn_skirtRC2_capsule, center=<6.6609 -7.0215 0>}]
 dyn_skirtLC4 [{shapeType=Capsule, length=6.25, radius=9.75, shapeName=dyn_skirtRC2_capsule, center=<6.6609 -7.0215 0>}]
 dyn_skirtMid1 [{shapeType=Capsule, length=3, radius=3, shapeName=dyn_skirtMid1_capsule}]
 dyn_skirtMid2 [{shapeType=Capsule, length=6.75, radius=9.25, shapeName=dyn_skirtMid2_capsule, center=<10 -3 0>, rotation=<rot 90 0 0>}]
 dyn_skirtMid3 [{shapeType=Capsule, length=21, radius=11.75, shapeName=dyn_skirtMid3_capsule, center=<10 -3 0>, rotation=<rot 0.0001 0 0>}]
 dyn_skirtMidL4 [{shapeType=Capsule, length=27.25, radius=12.5, shapeName=dyn_skirtMidL4_capsule, center=<25.7744 -3 -0.0006>}]
 dyn_skirtMidR4 [{shapeType=Capsule, length=27.25, radius=12.5, shapeName=dyn_skirtMidR4_capsule, center=<-27.4543 -3.0003 -0.0006>}]
 dyn_skirtRA1 [{shapeType=Capsule, length=3, radius=3, shapeName=dyn_skirtRA1_capsule}]
 dyn_skirtRA2 [{shapeType=Capsule, length=7.75, radius=8.25, shapeName=dyn_skirtLA2_capsule, center=<10.258 -4 0>, rotation=<rot 90 0 0>}]
 dyn_skirtRA3 [{shapeType=Capsule, length=7, radius=9, shapeName=dyn_skirtLA3_capsule, center=<10.258 -3 0>, rotation=<rot 90 0 0>}]
 dyn_skirtRA4 [{shapeType=Capsule, length=29.75, radius=10.25, shapeName=dyn_skirtLA4_capsule, center=<26.0324 -3 -0.0006>}]
 dyn_skirtRB1 [{shapeType=Capsule, length=3, radius=3, shapeName=dyn_skirtRB1_capsule}]
 dyn_skirtRB2 [{shapeType=Capsule, length=9, radius=7, shapeName=dyn_skirtRB2_capsule, center=<10 -7 0>, rotation=<rot 90 0 0>}]
 dyn_skirtRB3 [{shapeType=Capsule, length=8, radius=8, shapeName=dyn_skirtRB3_capsule, center=<10 -3 0>, rotation=<rot 90 0 0>}]
 dyn_skirtRB4 [{shapeType=Capsule, length=29, radius=10.75, shapeName=dyn_skirtRB4_capsule, center=<25.7744 -3 -0.0006>}]
 dyn_skirtRC1 [{shapeType=Capsule, length=10, radius=5, shapeName=dyn_skirtLC1_capsule}]
 dyn_skirtRC2 [{shapeType=Capsule, length=8, radius=8, shapeName=dyn_skirtRC2_capsule, center=<6.6609 -6.422 -0.0001>}]
 dyn_skirtRC3 [{shapeType=Capsule, length=6.25, radius=9.75, shapeName=dyn_skirtRC2_capsule, center=<6.6609 -7.0215 0>}]
 floorCOL [{shapeType=Capsule, length=0, radius=455.75, shapeName=floorCOL_capsule, center=<0 -40.0698 -457.7643>}]
 pelvis [{shapeType=Capsule, length=34.6625, radius=12.1087, shapeName=pelvis_capsule, center=<-60.1773 -0.001 0>, rotation=<rot -90 0.0099 -0.0099>}]
 root [{shapeType=Capsule, length=93.25, radius=167.5, shapeName=root_capsule, center=<17.235 -13.2737 -154.5117>, rotation=<rot 90 180 -109.995>}]
 thigh_l [{shapeType=Capsule, length=32.6942, radius=9.1238, shapeName=thigh_l_capsule, center=<-19 0 0>, rotation=<rot 90 0 0>}]
 thigh_r [{shapeType=Capsule, length=32.6942, radius=9.1238, shapeName=thigh_l_capsule, center=<19 0 0>, rotation=<rot 90 0 0>}]
}
constraints [
 {
  bone1Name=thigh_l
  bone2Name=pelvis
  swing1LimitDegrees=45
  swing1Motion=Free
  swing2LimitDegrees=45
  swing2Motion=Free
  twistLimitDegrees=45
  twistMotion=Free
 }
 {
  bone1Name=calf_l
  bone2Name=thigh_l
  swing1LimitDegrees=45
  swing1Motion=Free
  swing2LimitDegrees=45
  swing2Motion=Free
  twistLimitDegrees=45
  twistMotion=Free
 }
 {
  bone1Name=thigh_r
  bone2Name=pelvis
  swing1LimitDegrees=45
  swing1Motion=Free
  swing2LimitDegrees=45
  swing2Motion=Free
  twistLimitDegrees=45
  twistMotion=Free
 }
 {
  bone1Name=calf_r
  bone2Name=thigh_r
  swing1LimitDegrees=45
  swing1Motion=Free
  swing2LimitDegrees=45
  swing2Motion=Free
  twistLimitDegrees=45
  twistMotion=Free
 }
 {
  bone1Name=dyn_skirtLA1
  bone2Name=pelvis
  swing1LimitDegrees=45
  swing1Motion=Free
  swing2LimitDegrees=45
  swing2Motion=Free
  twistLimitDegrees=45
  twistMotion=Free
 }
 {bone1Name=dyn_skirtLA2, bone2Name=dyn_skirtLA1, swing1LimitDegrees=90, swing2LimitDegrees=90, twistLimitDegrees=0}
 {bone1Name=dyn_skirtLA3, bone2Name=dyn_skirtLA2, swing1LimitDegrees=90, swing2LimitDegrees=90, twistLimitDegrees=0}
 {bone1Name=dyn_skirtLA4, bone2Name=dyn_skirtLA3, swing1LimitDegrees=90, swing2LimitDegrees=90, twistLimitDegrees=0}
 {bone1Name=dyn_skirtMid1, bone2Name=pelvis, swing1LimitDegrees=90, swing2LimitDegrees=90, twistLimitDegrees=0}
 {bone1Name=dyn_skirtMid3, bone2Name=dyn_skirtMid2, swing1LimitDegrees=90, swing2LimitDegrees=90, twistLimitDegrees=0}
 {bone1Name=dyn_skirtMid2, bone2Name=dyn_skirtMid1, swing1LimitDegrees=90, swing2LimitDegrees=90, twistLimitDegrees=0}
 {bone1Name=dyn_skirtMidR4, bone2Name=dyn_skirtMid3, swing1LimitDegrees=90, swing2LimitDegrees=90, twistLimitDegrees=0}
 {bone1Name=dyn_skirtMidL4, bone2Name=dyn_skirtMid3, swing1LimitDegrees=90, swing2LimitDegrees=90, twistLimitDegrees=0}
 {bone1Name=dyn_skirtLB1, bone2Name=pelvis, swing1LimitDegrees=90, swing2LimitDegrees=90, twistLimitDegrees=0}
 {bone1Name=dyn_skirtLB2, bone2Name=dyn_skirtLB1, swing1LimitDegrees=90, swing2LimitDegrees=90, twistLimitDegrees=0}
 {bone1Name=dyn_skirtLB3, bone2Name=dyn_skirtLB2, swing1LimitDegrees=90, swing2LimitDegrees=90, twistLimitDegrees=0}
 {bone1Name=dyn_skirtLB4, bone2Name=dyn_skirtLB3, swing1LimitDegrees=90, swing2LimitDegrees=90, twistLimitDegrees=0}
 {bone1Name=dyn_skirtRC1, bone2Name=pelvis, swing1LimitDegrees=90, swing2LimitDegrees=90, twistLimitDegrees=0}
 {bone1Name=dyn_skirtRB1, bone2Name=pelvis, swing1LimitDegrees=90, swing2LimitDegrees=90, twistLimitDegrees=0}
 {bone1Name=dyn_skirtRB2, bone2Name=dyn_skirtRB1, swing1LimitDegrees=90, swing2LimitDegrees=90, twistLimitDegrees=0}
 {bone1Name=dyn_skirtRB3, bone2Name=dyn_skirtRB2, swing1LimitDegrees=90, swing2LimitDegrees=90, twistLimitDegrees=0}
 {bone1Name=dyn_skirtRB4, bone2Name=dyn_skirtRB3, swing1LimitDegrees=90, swing2LimitDegrees=90, twistLimitDegrees=0}
 {bone1Name=dyn_skirtRA1, bone2Name=pelvis, swing1LimitDegrees=90, swing2LimitDegrees=90, twistLimitDegrees=0}
 {bone1Name=dyn_skirtRA2, bone2Name=dyn_skirtRA1, swing1LimitDegrees=90, swing2LimitDegrees=90, twistLimitDegrees=0}
 {bone1Name=dyn_skirtRA3, bone2Name=dyn_skirtRA2, swing1LimitDegrees=45, swing2LimitDegrees=45, twistLimitDegrees=0}
 {bone1Name=dyn_skirtRA4, bone2Name=dyn_skirtRA3, swing1LimitDegrees=90, swing2LimitDegrees=90, twistLimitDegrees=0}
 {bone1Name=dyn_skirtRC2, bone2Name=dyn_skirtRC1, swing1LimitDegrees=45, swing2LimitDegrees=45, twistLimitDegrees=0}
 {
  bone1Name=dyn_skirtLC2
  bone2Name=pelvis
  swing1LimitDegrees=45
  swing1Motion=Free
  swing2LimitDegrees=45
  swing2Motion=Free
  twistLimitDegrees=45
  twistMotion=Free
 }
 {bone1Name=dyn_skirtLC3, bone2Name=dyn_skirtLC2, swing1LimitDegrees=45, swing2LimitDegrees=45, twistLimitDegrees=0}
 {
  bone1Name=pelvis
  bone2Name=root
  swing1LimitDegrees=45
  swing1Motion=Free
  swing2LimitDegrees=45
  swing2Motion=Free
  twistLimitDegrees=45
  twistMotion=Free
 }
 {
  bone1Name=floorCOL
  bone2Name=root
  swing1LimitDegrees=45
  swing1Motion=Free
  swing2LimitDegrees=45
  swing2Motion=Free
  twistLimitDegrees=45
  twistMotion=Free
 }
 {
  bone1Name=dyn_skirtRC3
  bone2Name=dyn_skirtRC2
  swing1LimitDegrees=45
  swing1Motion=Free
  swing2LimitDegrees=45
  swing2Motion=Free
  twistLimitDegrees=45
  twistMotion=Free
 }
 {
  bone1Name=dyn_skirtLC4
  bone2Name=dyn_skirtLC3
  swing1LimitDegrees=45
  swing1Motion=Free
  swing2LimitDegrees=45
  swing2Motion=Free
  twistLimitDegrees=45
  twistMotion=Free
 }
]
