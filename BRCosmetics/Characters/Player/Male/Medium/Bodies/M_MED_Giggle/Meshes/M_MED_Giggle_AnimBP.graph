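# Anim Blueprint for the M_MED_Giggle body cosmetic, animating skirt and chain secondary motion over the pose copied from the attached parent mesh.
# A Control Rig, thigh-following rotation constraints, leg-driven bone offsets and chain RigidBody simulation build a cached pose.
# That pose feeds RigidBody skirt sims using normal, skydive and front-end physics assets, then additive bone corrections.
# The EventGraph derives selectors and alphas from thigh rotation, crouch, speed, skydive and front-end state.
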AnimBlueprint /BRCosmetics/Characters/Player/Male/Medium/Bodies/M_MED_Giggle/Meshes/M_MED_Giggle_AnimBP
target_skeleton=&/M_MED_Giggle_Skeleton
graphs [
 AnimGraph {nodes [
   Local To Component()
   n1 = "Transform (Modify) Bone - Bone: dyn_skirtLC1"(Alpha=0, Translation="0,0,0", Rotation="-10,0,0", Scale="1,1,1") {
    rotationMode=BMM_Additive
    rotationSpace=BCS_BoneSpace
    scaleMode=BMM_Additive
    scaleSpace=BCS_BoneSpace
    translationMode=BMM_Additive
    translationSpace=BCS_BoneSpace
    alphaBoolBlend {blendOption=Linear}
    boneToModify {boneName=dyn_skirtLC1}
    rotation {pitch=-10}
   }
   "Transform (Modify) Bone - Bone: dyn_skirtRC1"(ComponentPose=@n1, Alpha=0, Translation="0,0,0", Rotation="10,0,0", Scale="1,1,1") {
    rotationMode=BMM_Additive
    rotationSpace=BCS_BoneSpace
    scaleMode=BMM_Additive
    scaleSpace=BCS_BoneSpace
    translationMode=BMM_Additive
    translationSpace=BCS_BoneSpace
    alphaBoolBlend {blendOption=Linear}
    boneToModify {boneName=dyn_skirtRC1}
    rotation {pitch=10}
   }
   n2 = RigidBody(ComponentPose=Local To Component(LocalPose=Use cached pose 'beforeBoolPose'())) {
    bClampLinearTranslationLimitToRefPose=true
    bOverrideWorldGravity=true
    overridePhysicsAsset=&/M_MED_Giggle__Skirt_PhysicsSky2
    simulationSpace=BaseBoneSpace
    alphaBoolBlend {blendOption=Linear}
    baseBoneRef {boneName=pelvis}
    componentAppliedLinearAccClamp=<3000 3000 3000>
    componentLinearAccScale=<0.45 0.35 0.35>
    componentLinearVelScale=<0.45 0.35 0.35>
    overrideWorldGravity {z=-980}
   }
   n3 = Component To Local(ComponentPose=@n2)
   n4 = RigidBody(ComponentPose=Local To Component(LocalPose=Use cached pose 'beforeBoolPose'()), Alpha=Get isNotFrontEnd() -> (isNotFrontEnd:double)) {
    bClampLinearTranslationLimitToRefPose=true
    bOverrideWorldGravity=true
    overridePhysicsAsset=&/M_MED_Giggle__Skirt_PhysicsNormal
    simulationSpace=BaseBoneSpace
    alphaBoolBlend {blendOption=Linear}
    baseBoneRef {boneName=pelvis}
    componentAppliedLinearAccClamp=<3000 3000 3000>
    componentLinearAccScale=<0.45 0.35 0.35>
    componentLinearVelScale=<0.45 0.35 0.35>
    overrideWorldGravity {z=-980}
   }
   n5 = Component To Local(ComponentPose=@n4)
   n6 = Get IsSkyDiveFloatingBool() -> (IsSkyDiveFloatingBool:bool)
   n7 = RigidBody(
    ComponentPose=Local To Component(LocalPose=Blend Poses by bool(BlendPose_0=@n3, BlendPose_1=@n5, bActiveValue=@n6) {blendType=Linear})
    Alpha=Get isFrontEnd() -> (isFrontEnd:double)
   ) {
    bClampLinearTranslationLimitToRefPose=true
    bOverrideWorldGravity=true
    overridePhysicsAsset=&/M_MED_Giggle__Skirt_PhysicsFrontEnd
    simulationSpace=BaseBoneSpace
    alphaBoolBlend {blendOption=Linear}
    baseBoneRef {boneName=pelvis}
    componentAppliedLinearAccClamp=<3000 3000 3000>
    componentLinearAccScale=<0.45 0.35 0.35>
    componentLinearVelScale=<0.45 0.35 0.35>
    overrideWorldGravity {z=-980}
   }
   n8 = "Transform (Modify) Bone - Bone: dyn_skirtRB1"(ComponentPose=@n7, Alpha=Get isFrontEnd(), Translation="0,-1,0", Rotation="0,0,0", Scale="1,1,1") {
    rotationMode=BMM_Additive
    rotationSpace=BCS_BoneSpace
    scaleMode=BMM_Additive
    scaleSpace=BCS_BoneSpace
    translationMode=BMM_Additive
    translationSpace=BCS_BoneSpace
    alphaBoolBlend {blendOption=Linear}
    alphaScaleBiasClamp {bInterpResult=true}
    boneToModify {boneName=dyn_skirtRB1}
    translation {y=-1}
   }
   n9 = "Transform (Modify) Bone - Bone: dyn_skirtRB2"(ComponentPose=@n8, Alpha=Get isFrontEnd(), Translation="0,-3,0", Rotation="0,0,0", Scale="1,1,1") {
    rotationMode=BMM_Additive
    rotationSpace=BCS_BoneSpace
    scaleMode=BMM_Additive
    scaleSpace=BCS_BoneSpace
    translationMode=BMM_Additive
    translationSpace=BCS_BoneSpace
    alphaBoolBlend {blendOption=Linear}
    alphaScaleBiasClamp {bInterpResult=true}
    boneToModify {boneName=dyn_skirtRB2}
    translation {y=-3}
   }
   n10 = "Transform (Modify) Bone - Bone: dyn_skirtRB4"(ComponentPose=@n9, Alpha=Get isFrontEnd(), Translation="0,0,0", Rotation="0,-18,0", Scale="1,1,1") {
    rotationMode=BMM_Additive
    rotationSpace=BCS_BoneSpace
    scaleMode=BMM_Additive
    scaleSpace=BCS_BoneSpace
    translationMode=BMM_Additive
    translationSpace=BCS_BoneSpace
    alphaBoolBlend {blendOption=Linear}
    alphaScaleBiasClamp {bInterpResult=true}
    boneToModify {boneName=dyn_skirtRB4}
    rotation {yaw=-18}
   }
   n11 = "Transform (Modify) Bone - Bone: dyn_skirtLC2"(ComponentPose=@n10, Alpha=Get isFrontEnd(), Translation="0,0,0", Rotation="0,-18,0", Scale="1,1,1") {
    rotationMode=BMM_Additive
    rotationSpace=BCS_BoneSpace
    scaleMode=BMM_Additive
    scaleSpace=BCS_BoneSpace
    translationMode=BMM_Additive
    translationSpace=BCS_BoneSpace
    alphaBoolBlend {blendOption=Linear}
    alphaScaleBiasClamp {bInterpResult=true}
    boneToModify {boneName=dyn_skirtLC2}
    rotation {yaw=-18}
   }
   n12 = "Transform (Modify) Bone - Bone: dyn_skirtLC1"(ComponentPose=@n11, Alpha=Get isFrontEnd(), Translation="0,0,0", Rotation="0,10,0", Scale="1,1,1") {
    rotationMode=BMM_Additive
    rotationSpace=BCS_BoneSpace
    scaleMode=BMM_Additive
    scaleSpace=BCS_BoneSpace
    translationMode=BMM_Additive
    translationSpace=BCS_BoneSpace
    alphaBoolBlend {blendOption=Linear}
    alphaScaleBiasClamp {bInterpResult=true}
    boneToModify {boneName=dyn_skirtLC1}
    rotation {yaw=10}
   }
   n13 = "Transform (Modify) Bone - Bone: dyn_hipChain2"(ComponentPose=@n12, Alpha=Get isFrontEnd(), Translation="0,-3.5,0", Rotation="0,10,0", Scale="1,1,1") {
    rotationMode=BMM_Additive
    rotationSpace=BCS_BoneSpace
    scaleMode=BMM_Additive
    scaleSpace=BCS_BoneSpace
    translationMode=BMM_Additive
    translationSpace=BCS_BoneSpace
    alphaBoolBlend {blendOption=Linear}
    alphaScaleBiasClamp {bInterpResult=true}
    boneToModify {boneName=dyn_hipChain2}
    rotation {yaw=10}
    translation {y=-3.5}
   }
   n14 = "Transform (Modify) Bone - Bone: dyn_skirtRC1"(ComponentPose=@n13, Alpha=1, Translation="0,0,0", Rotation="0,-10,0", Scale="1,1,1") {
    rotationMode=BMM_Additive
    rotationSpace=BCS_BoneSpace
    scaleMode=BMM_Additive
    scaleSpace=BCS_BoneSpace
    translationMode=BMM_Additive
    translationSpace=BCS_BoneSpace
    alphaBoolBlend {blendOption=Linear}
    alphaScaleBiasClamp {bInterpResult=true}
    boneToModify {boneName=dyn_skirtRC1}
    rotation {yaw=-10}
   }
   n15 = Get CrouchPose() -> (CrouchPose:double)
   n16 = "Transform (Modify) Bone - Bone: dyn_skirtRA1"(ComponentPose=@n14, Alpha=@n15, Translation="0,0,0", Rotation="5,-30,0", Scale="1,1,1") {
    rotationMode=BMM_Additive
    rotationSpace=BCS_BoneSpace
    scaleMode=BMM_Additive
    scaleSpace=BCS_BoneSpace
    translationMode=BMM_Additive
    translationSpace=BCS_BoneSpace
    alphaBoolBlend {blendOption=Linear}
    alphaScaleBiasClamp {bInterpResult=true}
    boneToModify {boneName=dyn_skirtRA1}
    rotation {pitch=5, yaw=-30}
   }
   n17 = "Transform (Modify) Bone - Bone: dyn_skirtRA2"(ComponentPose=@n16, Alpha=@n15, Translation="0,0,0", Rotation="-10,20,0", Scale="1,1,1") {
    rotationMode=BMM_Additive
    rotationSpace=BCS_BoneSpace
    scaleMode=BMM_Additive
    scaleSpace=BCS_BoneSpace
    translationMode=BMM_Additive
    translationSpace=BCS_BoneSpace
    alphaBoolBlend {blendOption=Linear}
    alphaScaleBiasClamp {bInterpResult=true}
    boneToModify {boneName=dyn_skirtRA2}
    rotation {pitch=-10, yaw=20}
   }
   n18 = "Transform (Modify) Bone - Bone: dyn_skirtMid1"(ComponentPose=@n17, Alpha=@n15, Translation="0,0,0", Rotation="-15,-30,0", Scale="1,1,1") {
    rotationMode=BMM_Additive
    rotationSpace=BCS_BoneSpace
    scaleMode=BMM_Additive
    scaleSpace=BCS_BoneSpace
    translationMode=BMM_Additive
    translationSpace=BCS_BoneSpace
    alphaBoolBlend {blendOption=Linear}
    alphaScaleBiasClamp {bInterpResult=true}
    boneToModify {boneName=dyn_skirtMid1}
    rotation {pitch=-15, yaw=-30}
   }
   n19 = "Transform (Modify) Bone - Bone: dyn_skirtMid2"(ComponentPose=@n18, Alpha=@n15, Translation="0,0,0", Rotation="0,20,0", Scale="1,1,1") {
    rotationMode=BMM_Additive
    rotationSpace=BCS_BoneSpace
    scaleMode=BMM_Additive
    scaleSpace=BCS_BoneSpace
    translationMode=BMM_Additive
    translationSpace=BCS_BoneSpace
    alphaBoolBlend {blendOption=Linear}
    alphaScaleBiasClamp {bInterpResult=true}
    boneToModify {boneName=dyn_skirtMid2}
    rotation {yaw=20}
   }
   n20 = Get isCrouchingSkirt() -> (isCrouchingSkirt:double)
   n21 = "Transform (Modify) Bone - Bone: dyn_skirtRC1"(ComponentPose=@n19, Alpha=@n20, Translation="0,0,0", Rotation="0,-5,0", Scale="1,1,1") {
    rotationMode=BMM_Additive
    rotationSpace=BCS_BoneSpace
    scaleMode=BMM_Additive
    scaleSpace=BCS_BoneSpace
    translationMode=BMM_Additive
    translationSpace=BCS_BoneSpace
    alphaBoolBlend {blendOption=Linear}
    alphaScaleBiasClamp {bInterpResult=true, interpSpeedDecreasing=1, interpSpeedIncreasing=1}
    boneToModify {boneName=dyn_skirtRC1}
    rotation {yaw=-5}
   }
   n22 = "Transform (Modify) Bone - Bone: dyn_skirtRB1"(ComponentPose=@n21, Alpha=@n20, Translation="0,-2,0", Rotation="0,0,0", Scale="1,1,1") {
    rotationMode=BMM_Additive
    rotationSpace=BCS_BoneSpace
    scaleMode=BMM_Additive
    scaleSpace=BCS_BoneSpace
    translationMode=BMM_Additive
    translationSpace=BCS_BoneSpace
    alphaBoolBlend {blendOption=Linear}
    alphaScaleBiasClamp {bInterpResult=true, interpSpeedDecreasing=100, interpSpeedIncreasing=100}
    boneToModify {boneName=dyn_skirtRB1}
    translation {y=-2}
   }
   n23 = Get Is Sky Dive Floating() -> (Is Sky Dive Floating:double)
   n24 = "Transform (Modify) Bone - Bone: dyn_skirtMid1"(ComponentPose=@n22, Alpha=@n23, Translation="0,0,0", Rotation="0,-30,0", Scale="1,1,1") {
    rotationMode=BMM_Additive
    rotationSpace=BCS_BoneSpace
    scaleMode=BMM_Additive
    scaleSpace=BCS_BoneSpace
    translationMode=BMM_Additive
    translationSpace=BCS_BoneSpace
    alphaBoolBlend {blendOption=Linear}
    alphaScaleBiasClamp {bInterpResult=true, interpSpeedIncreasing=5}
    boneToModify {boneName=dyn_skirtMid1}
    rotation {yaw=-30}
   }
   n25 = "Transform (Modify) Bone - Bone: dyn_skirtLB1"(ComponentPose=@n24, Alpha=@n23, Translation="0,0,0", Rotation="-20,-15,0", Scale="1,1,1") {
    rotationMode=BMM_Additive
    rotationSpace=BCS_BoneSpace
    scaleMode=BMM_Additive
    scaleSpace=BCS_BoneSpace
    translationMode=BMM_Additive
    translationSpace=BCS_BoneSpace
    alphaBoolBlend {blendOption=Linear}
    alphaScaleBiasClamp {bInterpResult=true, interpSpeedIncreasing=5}
    boneToModify {boneName=dyn_skirtLB1}
    rotation {pitch=-20, yaw=-15}
   }
   n26 = "Transform (Modify) Bone - Bone: dyn_skirtRB1"(ComponentPose=@n25, Alpha=@n23, Translation="0,0,0", Rotation="0,-10,0", Scale="1,1,1") {
    rotationMode=BMM_Additive
    rotationSpace=BCS_BoneSpace
    scaleMode=BMM_Additive
    scaleSpace=BCS_BoneSpace
    translationMode=BMM_Additive
    translationSpace=BCS_BoneSpace
    alphaBoolBlend {blendOption=Linear}
    alphaScaleBiasClamp {bInterpResult=true, interpSpeedIncreasing=5}
    boneToModify {boneName=dyn_skirtRB1}
    rotation {yaw=-10}
   }
   n27 = "Transform (Modify) Bone - Bone: dyn_skirtRA1"(ComponentPose=@n26, Alpha=@n23, Translation="0,0,0", Rotation="15,-25,0", Scale="1,1,1") {
    rotationMode=BMM_Additive
    rotationSpace=BCS_BoneSpace
    scaleMode=BMM_Additive
    scaleSpace=BCS_BoneSpace
    translationMode=BMM_Additive
    translationSpace=BCS_BoneSpace
    alphaBoolBlend {blendOption=Linear}
    alphaScaleBiasClamp {bInterpResult=true, interpSpeedIncreasing=5}
    boneToModify {boneName=dyn_skirtRA1}
    rotation {pitch=15, yaw=-25}
   }
   n28 = "Transform (Modify) Bone - Bone: dyn_skirtLA1"(ComponentPose=@n27, Alpha=@n23, Translation="0,0,0", Rotation="-15,-25,0", Scale="1,1,1") {
    rotationMode=BMM_Additive
    rotationSpace=BCS_BoneSpace
    scaleMode=BMM_Additive
    scaleSpace=BCS_BoneSpace
    translationMode=BMM_Additive
    translationSpace=BCS_BoneSpace
    alphaBoolBlend {blendOption=Linear}
    alphaScaleBiasClamp {bInterpResult=true, interpSpeedIncreasing=5}
    boneToModify {boneName=dyn_skirtLA1}
    rotation {pitch=-15, yaw=-25}
   }
   Output Pose(Result=Component To Local(ComponentPose=@n28)) {layerGroup=None}
   n29 = Control Rig!(
    Source=Copy Pose From Mesh() {bUseAttachedParent=true}
    bAlphaBoolEnabled:bool=True
    isCustomEm:bool=Get Is Custom Emote Playing() -> (Is Custom Emote Playing:bool)
   ) -> (Pose) {
    alphaInputType=Bool
    alphaBoolBlend {blendInTime=0.2, blendOutTime=0.2}
    controlRigAssetReference {blueprintRigClass=&/M_MED_Giggle_CtrlRig.M_MED_Giggle_CtrlRig_C}
   }
   n30 = "Constraint - dyn_skirtLC1 (RR)"!(ComponentPose=Local To Component(LocalPose=@n29), Alpha=1, ConstraintWeights_0=0, ConstraintWeights_1:float=1) -> (Pose) {
    alphaBoolBlend {blendOption=Linear}
    boneToModify {boneName=dyn_skirtLC1}
    constraintSetup [{transformType=Rotation, targetBone {boneName=pelvis}}, {transformType=Rotation, targetBone {boneName=thigh_l}}]
    constraintWeights [1,1]
   }
   n31 = "Constraint - dyn_skirtLB2 (RR)"!(ComponentPose=@n30, Alpha=1, ConstraintWeights_0=0, ConstraintWeights_1:float=1) -> (Pose) {
    alphaBoolBlend {blendOption=Linear}
    boneToModify {boneName=dyn_skirtLB2}
    constraintSetup [{transformType=Rotation, targetBone {boneName=pelvis}}, {transformType=Rotation, targetBone {boneName=thigh_l}}]
    constraintWeights [1,1]
   }
   n32 = "Constraint - dyn_skirtLA2 (RR)"!(ComponentPose=@n31, Alpha=1, ConstraintWeights_0=0.5, ConstraintWeights_1:float=0.5) -> (Pose) {
    alphaBoolBlend {blendOption=Linear}
    boneToModify {boneName=dyn_skirtLA2}
    constraintSetup [{transformType=Rotation, targetBone {boneName=pelvis}}, {transformType=Rotation, targetBone {boneName=thigh_l}}]
    constraintWeights [1,1]
   }
   n33 = "Constraint - dyn_skirtRC1 (RR)"!(ComponentPose=@n32, Alpha=1, ConstraintWeights_0=0, ConstraintWeights_1:float=1) -> (Pose) {
    alphaBoolBlend {blendOption=Linear}
    boneToModify {boneName=dyn_skirtRC1}
    constraintSetup [{transformType=Rotation, targetBone {boneName=pelvis}}, {transformType=Rotation, targetBone {boneName=thigh_r}}]
    constraintWeights [1,1]
   }
   n34 = "Constraint - dyn_skirtRB2 (RR)"!(ComponentPose=@n33, Alpha=1, ConstraintWeights_0=0, ConstraintWeights_1:float=1) -> (Pose) {
    alphaBoolBlend {blendOption=Linear}
    boneToModify {boneName=dyn_skirtRB2}
    constraintSetup [{transformType=Rotation, targetBone {boneName=pelvis}}, {transformType=Rotation, targetBone {boneName=thigh_r}}]
    constraintWeights [1,1]
   }
   n35 = "Constraint - dyn_skirtRA2 (RR)"!(ComponentPose=@n34, Alpha=1, ConstraintWeights_0=0.5, ConstraintWeights_1:float=0.5) -> (Pose) {
    alphaBoolBlend {blendOption=Linear}
    boneToModify {boneName=dyn_skirtRA2}
    constraintSetup [{transformType=Rotation, targetBone {boneName=pelvis}}, {transformType=Rotation, targetBone {boneName=thigh_r}}]
    constraintWeights [1,1]
   }
   n36 = "Transform (Modify) Bone - Bone: dyn_skirtLC1"(ComponentPose=@n35, Alpha=Get leg_l_rotation fwd() -> (leg_l_rotation fwd:double), Translation="0,0,0", Rotation="-40,0,7", Scale="1,1,1") {
    rotationMode=BMM_Additive
    rotationSpace=BCS_BoneSpace
    scaleMode=BMM_Additive
    scaleSpace=BCS_BoneSpace
    translationMode=BMM_Additive
    translationSpace=BCS_BoneSpace
    alphaBoolBlend {blendOption=Linear}
    boneToModify {boneName=dyn_skirtLC1}
    rotation {pitch=-40, roll=7}
   }
   n37 = "Transform (Modify) Bone - Bone: dyn_skirtRC1"(ComponentPose=@n36, Alpha=Get leg_r_rotation_fwd() -> (leg_r_rotation_fwd:double), Translation="0,0,0", Rotation="40,0,0", Scale="1,1,1") {
    rotationMode=BMM_Additive
    rotationSpace=BCS_BoneSpace
    scaleMode=BMM_Additive
    scaleSpace=BCS_BoneSpace
    translationMode=BMM_Additive
    translationSpace=BCS_BoneSpace
    alphaBoolBlend {blendOption=Linear}
    boneToModify {boneName=dyn_skirtRC1}
    rotation {pitch=40}
   }
   n38 = "Transform (Modify) Bone - Bone: dyn_hipChain"(ComponentPose=@n37, Alpha=Get leg_l_rotation fwd(), Translation="0,0,0", Rotation="0,-17,0", Scale="1,1,1") {
    rotationMode=BMM_Additive
    rotationSpace=BCS_BoneSpace
    scaleMode=BMM_Additive
    scaleSpace=BCS_BoneSpace
    translationMode=BMM_Additive
    translationSpace=BCS_BoneSpace
    alphaBoolBlend {blendOption=Linear}
    boneToModify {boneName=dyn_hipChain}
    rotation {yaw=-17}
   }
   n39 = RigidBody(ComponentPose=@n38) {
    bClampLinearTranslationLimitToRefPose=true
    bOverrideWorldGravity=true
    overridePhysicsAsset=&/M_MED_Giggle__Chains_Physics
    simulationSpace=BaseBoneSpace
    alphaBoolBlend {blendOption=Linear}
    baseBoneRef {boneName=spine_01}
    componentAppliedLinearAccClamp=<3000 3000 3000>
    componentLinearAccScale=<0.4 0.3 0.3>
    componentLinearVelScale=<0.4 0.3 0.3>
    overrideWorldGravity {z=-980}
   }
   Save cached pose 'beforeBoolPose'(Pose=Component To Local(ComponentPose=@n39)) {cacheName=beforeBoolPose}
  ]}
 EventGraph {nodes [
   Event BlueprintUpdateAnimation()
   > n1 = Sequence!(execute) -> (then_0, then_1, then_2:Exec, then_3:Exec, then_4:Exec)
   n2 = GetOwnerSkeletalMeshComponent()
   > Set leg_l_rotation fwd(leg_l_rotation fwd:double=GetAlphaFromReferenceTransformDelta(SkeletalMeshComponent=@n2, BoneName=thigh_l, BaseName=pelvis, InputRange=(X=-20,Y=-100), TwistAxis="1,0,0")) -> (Output_Get:double)
   > Set leg_r_rotation_fwd(leg_r_rotation_fwd:double=GetAlphaFromReferenceTransformDelta(SkeletalMeshComponent=@n2, BoneName=thigh_r, BaseName=pelvis, InputRange=(X=-20,Y=-100), TwistAxis="1,0,0")) -> (Output_Get:double)
   > Set leg_l_rotation side(leg_l_rotation side:double=GetAlphaFromReferenceTransformDelta(BoneName=thigh_l, BaseName=pelvis, DeltaComponentType=EulerY, InputRange=(X=20,Y=100), OutputRange=(X=1,Y=0), TwistAxis="1,0,0")) -> (Output_Get:double)
   > Set leg_r_rotation side(leg_r_rotation side:double=GetAlphaFromReferenceTransformDelta(BoneName=thigh_r, BaseName=pelvis, DeltaComponentType=EulerY, InputRange=(X=20,Y=100), OutputRange=(X=1,Y=0), TwistAxis="1,0,0")) -> (Output_Get:double)
   n3 = GetOwnerSkeletalMeshComponent()
   > Set Is Custom Emote Playing(Is Custom Emote Playing:bool=Montage_IsPlaying(self=GetAnimInstance(self=@n3), Montage=/BRCosmetics/Animation/Game/MainPlayer/Emotes/ForestJeanGrowl/Emote_ForestJeanGrowl_CMM_FE_M)) -> (Output_Get:bool)
   n4 = Branch(execute=@n1.then_1, Condition="float > float"(A=Get PawnSpeed2D(), B=50) {is=title K2Node_CallFunction})
   > n5 = Set RunAlpha(RunAlpha:bool=false) -> (Output_Get:bool)
   n6 = Get bIsCrouching()
   n7 = Get RunAlpha() -> (RunAlpha:bool)
   Set CrouchPose(execute=[@n5.then, @n8.then], CrouchPose:double="To Float (Boolean)"(InBool=AND Boolean(A=@n6, B=@n7))) -> (Output_Get:double)
   n8 = Set RunAlpha(execute=@n4.else, RunAlpha=true)
   n9 = NOT Boolean(A=Get bIsSkydiveFloating())
   n10 = NOT Boolean(A=Get bIsSkydiveDiveMode())
   n11 = Get bIsSkyDiving()
   Set Is Sky Dive Floating(execute=@n1.then_2, Is Sky Dive Floating:double="To Float (Boolean)"(InBool=AND Boolean!(A=@n11, B=@n9, C:bool=@n10) -> (ReturnValue))) -> (Output_Get:double)
   n12 = Get bIsSkydiveFloating()
   n13 = Get bIsSkydiveDiveMode()
   > Set IsSkyDiveFloatingBool(IsSkyDiveFloatingBool:bool=AND Boolean(A=Get bIsSkyDiving(), B=NOT Boolean(A=@n12), C=NOT Boolean(A=@n13))) -> (Output_Get:bool)
   n14 = Branch(execute=@n1.then_3, Condition=Get bIsInFrontEnd())
   > Set isFrontEnd(isFrontEnd:double=1) -> (Output_Get:double)
   Set isNotFrontEnd(execute=@n14.else, isNotFrontEnd:double=1) -> (Output_Get:double)
   n15 = Branch(execute=@n1.then_4, Condition=Get bIsCrouching())
   > Set isCrouchingSkirt(isCrouchingSkirt:double=1) -> (Output_Get:double)
   Set isCrouchingSkirt(execute=@n15.else, isCrouchingSkirt=0)
  ]}
]
variables [
 CrouchPose {type=number}
 Is Crouching {type=boolean}
 Is Crouching Ev {type=boolean}
 Is Custom Emote Playing {type=boolean}
 Is Sky Dive Floating {type=number}
 isCrouchingSkirt {type=number}
 isFrontEnd {type=number}
 isNotFrontEnd {type=number}
 IsSkyDiveFloatingBool {type=boolean}
 IsSkydivingFloat {type=number}
 leg_l_rotation fwd {type=number}
 leg_l_rotation side {type=number}
 leg_r_rotation side {type=number}
 leg_r_rotation_fwd {type=number}
 RunAlpha {type=boolean}
]
