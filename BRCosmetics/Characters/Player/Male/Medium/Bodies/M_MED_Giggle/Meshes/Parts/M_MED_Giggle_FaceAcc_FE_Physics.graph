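# Physics asset for the face accessory mesh part of the Giggle male medium player body, covering its dynamic collar and head bones.
# Capsule bodies sit on two-bone left and right collar chains, head, head_dyn and spine_05.
# Each collar root, head and head_dyn is constrained to spine_05, and each collar tip to its root.
# Most joints rotate freely, but the right collar tip and head_dyn differ, with zero twist limit.
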PhysicsAsset /BRCosmetics/Characters/Player/Male/Medium/Bodies/M_MED_Giggle/Meshes/Parts/M_MED_Giggle_FaceAcc_FE_Physics
bodies {
 dyn_collarL1 [{
   shapeType=Capsule
   length=0.505
   radius=0.505
   shapeName=dyn_collarL1_capsule
   center=<6.1887 -1.9734 -1.598>
   rotation=<rot 0 -134.5092 -77.7336>
  }]
 dyn_collarL2 [{shapeType=Capsule, length=2, radius=2, shapeName=dyn_collarL2_capsule, center=<1.1243 0 0>, rotation=<rot 90 0 0>}]
 dyn_collarR1 [{shapeType=Capsule, length=0.505, radius=0.505, shapeName=dyn_collarR1_capsule, rotation=<rot 0 -90 0>}]
 dyn_collarR2 [{shapeType=Capsule, length=2, radius=2, shapeName=dyn_collarR2_capsule, center=<-1.4059 0 0>, rotation=<rot 90 0 0>}]
 head [{shapeType=Capsule, length=55.3226, radius=10.8124, shapeName=head_capsule, center=<13.8978 11.6493 0.0001>, rotation=<rot 0 -90 0>}]
 head_dyn [{shapeType=Capsule, length=12.8754, radius=5.6133, shapeName=head_dyn_capsule, center=<0 -15 0>}]
 spine_05 [{shapeType=Capsule, length=15.1009, radius=10.8469, shapeName=spine_05_capsule}]
}
constraints [
 {
  bone1Name=dyn_collarL1
  bone2Name=spine_05
  swing1LimitDegrees=45
  swing1Motion=Free
  swing2LimitDegrees=45
  swing2Motion=Free
  twistLimitDegrees=45
  twistMotion=Free
 }
 {
  bone1Name=dyn_collarL2
  bone2Name=dyn_collarL1
  swing1LimitDegrees=45
  swing1Motion=Free
  swing2LimitDegrees=45
  swing2Motion=Free
  twistLimitDegrees=45
  twistMotion=Free
 }
 {
  bone1Name=dyn_collarR1
  bone2Name=spine_05
  swing1LimitDegrees=45
  swing1Motion=Free
  swing2LimitDegrees=45
  swing2Motion=Free
  twistLimitDegrees=45
  twistMotion=Free
 }
 {bone1Name=dyn_collarR2, bone2Name=dyn_collarR1, swing1LimitDegrees=45, swing2LimitDegrees=45, twistLimitDegrees=0}
 {
  bone1Name=head
  bone2Name=spine_05
  swing1LimitDegrees=45
  swing1Motion=Free
  swing2LimitDegrees=45
  swing2Motion=Free
  twistLimitDegrees=45
  twistMotion=Free
 }
 {bone1Name=head_dyn, bone2Name=spine_05, swing1LimitDegrees=90, swing2LimitDegrees=90, twistLimitDegrees=0}
]
