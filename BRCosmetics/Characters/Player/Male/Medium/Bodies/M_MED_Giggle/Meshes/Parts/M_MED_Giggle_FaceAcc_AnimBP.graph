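# Post-copy AnimBP for the M_MED_Giggle face accessory, adding physics secondary motion to the pose copied from its attached parent mesh.
# Two spine_05-relative RigidBody nodes, one per physics asset, are switched by alphas derived from the front-end flag.
# In the front end, additive offsets also adjust dyn_collarL1 and head_dyn.
# The event graph also computes head-versus-spine alphas, but the head_dyn correctives they drive are disconnected from the output pose.
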AnimBlueprint /BRCosmetics/Characters/Player/Male/Medium/Bodies/M_MED_Giggle/Meshes/Parts/M_MED_Giggle_FaceAcc_AnimBP
target_skeleton=&/M_MED_Giggle_FaceAcc_Skeleton
graphs [
 AnimGraph {nodes [
   n1 = "Transform (Modify) Bone - Bone: head_dyn"(Alpha=Get headRotBack() -> (headRotBack:double), Translation="0,0,0", Rotation="0,-10,0", Scale="1,1,1") {
    rotationMode=BMM_Additive
    rotationSpace=BCS_BoneSpace
    scaleMode=BMM_Additive
    scaleSpace=BCS_BoneSpace
    translationMode=BMM_Additive
    translationSpace=BCS_BoneSpace
    alphaBoolBlend {blendOption=Linear}
    alphaScaleBiasClamp {bInterpResult=true, interpSpeedDecreasing=20, interpSpeedIncreasing=20}
    boneToModify {boneName=head_dyn}
    rotation {yaw=-10}
   }
   "Transform (Modify) Bone - Bone: head_dyn"(ComponentPose=@n1, Alpha=Get headBackxTwist() -> (headBackxTwist:double), Translation="-3,0,0", Rotation="0,0,0", Scale="1,1,1") {
    rotationMode=BMM_Additive
    rotationSpace=BCS_BoneSpace
    scaleMode=BMM_Additive
    scaleSpace=BCS_BoneSpace
    translationMode=BMM_Additive
    translationSpace=BCS_BoneSpace
    alphaBoolBlend {blendOption=Linear}
    alphaScaleBiasClamp {bInterpResult=true, interpSpeedDecreasing=20, interpSpeedIncreasing=20}
    boneToModify {boneName=head_dyn}
    translation {x=-3}
   }
   n2 = RigidBody(
    ComponentPose=Local To Component(LocalPose=Copy Pose From Mesh() {bUseAttachedParent=true})
    Alpha=Get Not_FrontEnd() -> (Not_FrontEnd:double)
   ) {
    bClampLinearTranslationLimitToRefPose=true
    overridePhysicsAsset=&/M_MED_Giggle_FaceAcc_Physics
    simulationSpace=BaseBoneSpace
    alphaBoolBlend {blendOption=Linear}
    baseBoneRef {boneName=spine_05}
    componentAppliedLinearAccClamp=<3000 3000 3000>
    componentLinearAccScale=<0.4 0.3 0.3>
    componentLinearVelScale=<0.4 0.3 0.3>
   }
   n3 = RigidBody(ComponentPose=@n2, Alpha=Get isFrontEnd() -> (isFrontEnd:double)) {
    bClampLinearTranslationLimitToRefPose=true
    overridePhysicsAsset=&/M_MED_Giggle_FaceAcc_FE_Physics
    simulationSpace=BaseBoneSpace
    alphaBoolBlend {blendOption=Linear}
    baseBoneRef {boneName=spine_05}
    componentAppliedLinearAccClamp=<3000 3000 3000>
    componentLinearAccScale=<0.4 0.3 0.3>
    componentLinearVelScale=<0.4 0.3 0.3>
   }
   n4 = Get isFrontEnd()
   n5 = "Transform (Modify) Bone - Bone: dyn_collarL1"(ComponentPose=@n3, Alpha=@n4, Translation="0,0,3", Rotation="0,0,0", Scale="1,1,1") {
    rotationMode=BMM_Additive
    rotationSpace=BCS_BoneSpace
    scaleMode=BMM_Additive
    scaleSpace=BCS_BoneSpace
    translationMode=BMM_Additive
    translationSpace=BCS_BoneSpace
    alphaBoolBlend {blendOption=Linear}
    alphaScaleBiasClamp {bInterpResult=true, interpSpeedDecreasing=20, interpSpeedIncreasing=20}
    boneToModify {boneName=dyn_collarL1}
    translation {z=3}
   }
   n6 = "Transform (Modify) Bone - Bone: head_dyn"(ComponentPose=@n5, Alpha=@n4, Translation="-1,-2,0", Rotation="-0.000,-15,-0.000", Scale="1,1,1") {
    rotationMode=BMM_Additive
    rotationSpace=BCS_BoneSpace
    scaleMode=BMM_Additive
    scaleSpace=BCS_BoneSpace
    translationMode=BMM_Additive
    translationSpace=BCS_BoneSpace
    alphaBoolBlend {blendOption=Linear}
    alphaScaleBiasClamp {bInterpResult=true, interpSpeedDecreasing=20, interpSpeedIncreasing=20}
    boneToModify {boneName=head_dyn}
    rotation {yaw=-15}
    translation=<-1 -2>
   }
   Output Pose(Result=Component To Local(ComponentPose=@n6)) {layerGroup=None}
  ]}
 EventGraph {nodes [
   Event BlueprintUpdateAnimation()
   n1 = GetOwnerSkeletalMeshComponent()
   n2 = GetAlphaFromReferenceTransformDelta(SkeletalMeshComponent=@n1, BoneName=head, BaseName=spine_05, InputRange=(X=20,Y=-50), TwistAxis="1,0,0")
   > Set headRotBack(headRotBack:double=@n2) -> (Output_Get:double)
   > Set headBackxTwist(headBackxTwist:double=float * float(
     A=@n2
     B=GetAlphaFromReferenceTransformDelta(SkeletalMeshComponent=@n1, BoneName=head, BaseName=spine_05, DeltaComponentType=EulerX, InputRange=(X=0,Y=-50), TwistAxis="1,0,0")
    )) -> (Output_Get:double)
   > Set isFrontEnd(isFrontEnd:double="To Float (Boolean)"(InBool=Get bIsInFrontEnd())) -> (Output_Get:double)
   > Set Not_FrontEnd(Not_FrontEnd:double=SelectFloat(A=0, B=1, bPickA=Get bIsInFrontEnd())) -> (Output_Get:double)
  ]}
]
variables [
 headBackxTwist {type=number}
 headRotBack {type=number}
 isFrontEnd {type=number}
 Not_FrontEnd {type=number}
]
